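# Niagara effect for the Button Case backpack cosmetic that spawns short-lived, gold-tinted star sprites on its skeletal mesh.
# A single Stars emitter spawns intermittently, varies color by time of day, and pops sprites up in size before shrinking them away.
# Particle update re-samples the mesh for position, while a second, non-positioning sample anchors two inverted, oppositely offset region masks that multiply into alpha to hide stars by region.
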NiagaraSystem /BRCosmetics/Accessories/FORT_Backpacks/Backpack_M_MED_ButtonCase/FX/NS_Backpack_ButtonCase
dependencies {
 usedDataInterfaces [&NiagaraDataInterfaceCurve, &NiagaraDataInterfaceSkeletalMesh]
 usedDynamicInputs []
 usedModules [
  &self:SystemSpawnScript.SystemScriptSource.SystemScriptGraph.NiagaraNodeAssignment_0.SetVariables_#1
  &SystemState
 ]
}
emitters {Stars {
  emitterUpdateScript [
   &EmitterState(Life Cycle Mode=::NewEnumerator1, Loop Duration=3, UseLoopDelay=false)
   &SpawnRate(Spawn Probability=0.6, SpawnRate=4, Use Spawn Probability=true)
  ]
  particleSpawnScript [
   &InitializeParticle(
    Color=&FNTimeOfDayColor(Night Color=<rgba 1 0.6005 0 2>, Morning Color=<rgba 2 1.2009 0 2>, Day Color=<rgba 2 1.2009 0 2>, Evening Color=<rgba 2 1.2009 0 0.9>)
    Lifetime Max=1.3
    Lifetime Mode=::NewEnumerator1
    Sprite Size Mode=::NewEnumerator1
    Uniform Sprite Size Max=25
    Uniform Sprite Size Min=15
    UsePositionOffset=false
    Write Material Random=true
   )
   &CameraOffset(Camera Offset Amount=15)
   &Color(Color=&MultiplyLinearColor(
     Linear Color=$Particles.Color
     Scale Factor=&MakeVector4_FromLinearColor(LinearColor=&FNTimeOfDayColor(Night Color=<rgba 0.3611 0.3761 0.6979 1>, Morning Color=<rgba 1 0.717 0.7469 1>, Day Color=<rgba 1 1 1 1>, Evening Color=<rgba 1 1 1 1>))
    ))
  ]
  particleUpdateScript [
   &SkeletalMeshLocation(
    MaskSpawnGroup=false
    Mesh Sampling Type=::NewEnumerator8
    OffsetAlongNormal=false
    OverrideRandomSeed=false
    Skeletal Mesh=$System.Skeletal Mesh
   )
   &SkeletalMeshLocation(
    MaskSpawnGroup=false
    Mesh Sampling Type=::NewEnumerator5
    Skeletal Mesh=$System.Skeletal Mesh
    Socket Sampling Mode=::NewEnumerator5
    Write To Particle Position=false
   ) {module=SkeletalMeshLocation001}
   &RegionMask(
    Invert Region=true
    Offset:Vector3f=<-3 0 -10>
    Region Dimensions:Vector3f=<25 100 70>
    Region Origin:Vector3f=$Particles.SkeletalMeshLocation001.SampledPosition
    Region Type:ENiagaraDebugDrawShape=::NewEnumerator1
    UseOffset=true
    UseRotation=false
   )
   &RegionMask(
    Invert Region=true
    Offset:Vector3f=<3 0 -10>
    Region Dimensions:Vector3f=<25 100 70>
    Region Origin:Vector3f=$Particles.SkeletalMeshLocation001.SampledPosition
    Region Type:ENiagaraDebugDrawShape=::NewEnumerator1
    UseOffset=true
    UseRotation=false
   ) {module=RegionMask001}
   &ParticleState()
   &ScaleSpriteSize(Uniform Curve Sprite Scale=DataInterface() {propertyValues {Curve {keys [
        {interpMode=RCIM_Cubic, leaveTangent=11.6533, tangentMode=RCTM_Break, time=0, value=0}
        {interpMode=RCIM_Cubic, leaveTangent=-4.0415, tangentMode=RCTM_Break, time=0.2, value=1}
        {tangentMode=RCTM_User, time=1, value=0}
       ]}}})
   &ScaleColor(Scale Alpha=$Particles.RegionMask.Float, ScaleA=true, ScaleRGB=false)
   &ScaleColor(Scale Alpha=$Particles.RegionMask001.Float, ScaleA=true, ScaleRGB=false) {module=ScaleColor001}
  ]
  rendererClasses [&NiagaraSpriteRendererProperties]
  renderers [{rendererClass=&NiagaraSpriteRendererProperties}]
 }}
summary {
 systemName=NS_Backpack_ButtonCase
 emitters [{emitterName=Stars, rendererClasses [&NiagaraSpriteRendererProperties]}]
 userVariables []
}
